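# M_DrawMin is an unlit, masked helper material for the custom mip generator in TechArtTools.
# It passes a Tex texture object parameter and a Level input into the MF_MinReduce material function, which, judging by its name, performs a minimum-value reduction.
# All sampling and reduction logic lives in that function, leaving this material as a thin wrapper.
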Material /TechArtTools/CustomMipGenerator/HelperMaterials/M_DrawMin
graph {
 properties {
  bAutomaticallySetUsageInEditor=false
  bUsedWithEditorCompositing=true
  blendMode=BLEND_Masked
  shadingModel=MSM_Unlit
  usedShadingModels=Unlit
 }
 nodes [MaterialFunctionCall!(Tex=TextureObjectParameter() {parameterName=Tex}, Level) {materialFunction=&/MF_MinReduce}]
}
