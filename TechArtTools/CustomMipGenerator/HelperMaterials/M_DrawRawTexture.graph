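# Unlit masked helper material for the custom mip generator tool that, going by its name, draws a texture's raw texels.
# The Tex parameter feeds both a texture size lookup and the sample, with PointSampledUVs using that size to snap UVs to texels.
# Sampling is pinned to mip 0 with view mip bias off.
# X and Y parameters form a Vector2 that no other shown node reads.
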Material /TechArtTools/CustomMipGenerator/HelperMaterials/M_DrawRawTexture
graph {
 expressions {
  MaterialExpressionConstant_0 {class=MaterialExpressionConstant}
  MaterialExpressionConvert_0 {
   class=MaterialExpressionConvert
   inputs [
    None {from=MaterialExpressionScalarParameter_0}
    None {from=MaterialExpressionScalarParameter_1}
   ]
   properties {
    nodeName=Make Vector2
    convertInputs [
     {
      type=Scalar
      expressionInput {
       expression=&self:MaterialExpressionScalarParameter_0
       inputName=None
       mask=0
       maskA=0
       maskB=0
       maskG=0
       maskR=0
       outputIndex=0
      }
     }
     {
      type=Scalar
      expressionInput {
       expression=&self:MaterialExpressionScalarParameter_1
       inputName=None
       mask=0
       maskA=0
       maskB=0
       maskG=0
       maskR=0
       outputIndex=0
      }
     }
    ]
    convertOutputs [{type=Vector2}]
   }
  }
  MaterialExpressionMaterialFunctionCall_0 {
   class=MaterialExpressionMaterialFunctionCall
   inputs [
    UVs {from=""}
    Texture Size {from=MaterialExpressionTextureProperty_0}
   ]
   properties {materialFunction=/Engine/Functions/Engine_MaterialFunctions01/Texturing/PointSampledUVs}
  }
  MaterialExpressionScalarParameter_0 {class=MaterialExpressionScalarParameter, properties {defaultValue=1, parameterName=X}}
  MaterialExpressionScalarParameter_1 {class=MaterialExpressionScalarParameter, properties {defaultValue=1, parameterName=Y}}
  MaterialExpressionStaticBool_0 {class=MaterialExpressionStaticBool}
  MaterialExpressionTextureObjectParameter_0 {
   class=MaterialExpressionTextureObjectParameter
   inputs [
    UVs {from="", output=""}
    Apply View MipBias {from="", output=""}
   ]
   properties {parameterName=Tex, samplerType=SAMPLERTYPE_LinearGrayscale, texture=/Engine/EngineMaterials/FastBlueNoise_scalar_128x128x64}
  }
  MaterialExpressionTextureProperty_0 {class=MaterialExpressionTextureProperty, inputs [Tex {from=MaterialExpressionTextureObjectParameter_0, output=""}], properties {}}
  MaterialExpressionTextureSample_0 {
   class=MaterialExpressionTextureSample
   inputs [
    UVs {from=MaterialExpressionMaterialFunctionCall_0, output=Result}
    Tex {from=MaterialExpressionTextureObjectParameter_0}
    Level {from=MaterialExpressionConstant_0}
    Apply View MipBias {from=MaterialExpressionStaticBool_0}
   ]
   properties {mipValueMode=TMVM_MipLevel}
  }
 }
 properties {
  bAutomaticallySetUsageInEditor=false
  bUsedWithEditorCompositing=true
  blendMode=BLEND_Masked
  shadingModel=MSM_Unlit
  usedShadingModels=Unlit
 }
}
