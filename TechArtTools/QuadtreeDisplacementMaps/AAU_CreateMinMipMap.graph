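# Editor utility that builds a max-filtered mip chain from the selected Texture2D and saves it as a new "_QuadtreeDisplacement" texture asset.
# GetFormat picks a render target format and temporary source compression, then one target is created per mip down to 1x1.
# The raw texture fills mip 0, and M_DrawMax (stored as MinMat) draws each level from the previous.
# The targets become the asset's mips before compression and sRGB are set.
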EditorUtilityBlueprint /TechArtTools/QuadtreeDisplacementMaps/AAU_CreateMinMipMap
graphs [
 CreateMaximumMipMapTexture {nodes [
   n1 = Get RTs() -> (RTs:Array of Texture Render Target 2D Object References)
   Reverse!(execute, TargetArray:Array of Texture Render Target 2D Object References=@n1) -> (then) {is=title K2Node_CallArrayFunction}
   > n2 = Reverse(TargetArray:Array of Int Vector 2 Structures=Get resarray() -> (resarray:Array of Int Vector 2 Structures))
   n3 = "Get (a copy)"(Array:Array of Texture Render Target 2D Object References=@n1) -> (Output:Texture Render Target 2D Object Reference)
   PrintString(
    execute=[@n2.then, @n16.then_2]
    InString="To String (Integer)"(InInt=Get SizeX!(self:Texture Render Target 2D Object Reference=@n3) -> (SizeX:int=0))
   )
   > PrintString(InString="To String (Integer)"(InInt=Get SizeY!(self:Texture Render Target 2D Object Reference=@n3) -> (SizeY:int=0)))
   n4 = Get Object Path String(Object=Get Texture2d() -> (Texture2d:Texture 2D Object Reference))
   n5 = GetEditorWorld(execute=@n36.then, self=Get UnrealEditorSubsystem())
   n6 = Create Texture Asset from Render Target Mips!(
    execute=@n16.then_5
    WorldContextObject:Object Reference=@n5
    RenderTargets:Array of Texture Render Target 2D Object References=Get RTs()
    PackagePathAndName:String=Append(A=Split(SourceString=@n4, InStr=., SearchDir=FromStart), B=_QuadtreeDisplacement)
   ) -> (then, bSuccess:bool, OutErrorMessage:String, ReturnValue:Texture 2D Object Reference)
   n7 = PrintString(InString=@n6.OutErrorMessage)
   SetEditorProperty(
    execute=[@n40.then, @n7.then, @n37.OutputPin]
    Object=@n6.ReturnValue
    PropertyName=CompressionSettings
    PropertyValue:TextureCompressionSettings Enum=Get set() -> (set:TextureCompressionSettings Enum)
   )
   > SetEditorProperty(
    Object=@n6.ReturnValue
    PropertyName=LossyCompressionAmount
    PropertyValue:Byte="To Byte (Integer)"!(InInt:int=MakeLiteralInt!(Value:int=1) -> (ReturnValue:int=0)) -> (ReturnValue:Byte=0)
   )
   > Set SRGB(SRGB=Get srgb() -> (srgb:bool), self=@n6.ReturnValue)
   > Is Valid(InputObject=@n6.ReturnValue)
   > SaveLoadedAsset(AssetToSave=@n6.ReturnValue, bOnlyIfIsDirty=false)
   n8 = Get Texture2d()
   MarkPropertyDirty!(execute, Object:Object Reference=@n8, PropertyName:Name=CompressionSettings) -> (then)
   n9 = Get Texture2d()
   > Set SRGB(SRGB=true, self=@n9)
   > Set NeverStream!(execute, NeverStream:bool=true, self:Streamable Render Asset Object Reference=@n9) -> (then, Output_Get:bool)
   > Set CompressFinal!(execute, CompressFinal:bool=false, self:Texture Object Reference=@n9) -> (then, Output_Get:bool)
   > Set CompressFinal(CompressFinal=true, self=@n9)
   n10 = CreateDynamicMaterialInstance(WorldContextObject=@n5, Parent=/TechArtTools/CustomMipGenerator/HelperMaterials/M_DrawMin, OptionalName=None, CreationFlags=Transient)
   n11 = CreateDynamicMaterialInstance(
    execute=[@n10.then, @n13.then]
    WorldContextObject=@n5
    Parent=/TechArtTools/CustomMipGenerator/HelperMaterials/M_DrawRawTexture
    OptionalName=None
    CreationFlags=Transient
   )
   > n12 = Set rawMat(rawMat:Material Instance Dynamic Object Reference=@n11) -> (Output_Get:Material Instance Dynamic Object Reference)
   n13 = CreateDynamicMaterialInstance(
    execute=@n5.then
    WorldContextObject=@n5
    Parent=/TechArtTools/CustomMipGenerator/HelperMaterials/M_DrawMax
    OptionalName=None
    CreationFlags=Transient
   )
   Set MinMat(execute=@n12.then, MinMat:Material Instance Dynamic Object Reference=@n13) -> (Output_Get:Material Instance Dynamic Object Reference)
   > n14 = GetSelectedAssets()
   > n15 = Cast To Texture2D(Object="Get (a copy)"(Array:Array of Object References=@n14) -> (Output:Object Reference))
   > PrintString(InString=GetDisplayName(Object=@n15))
   > Set Texture2d(Texture2d:Texture 2D Object Reference=@n15) -> (Output_Get:Texture 2D Object Reference)
   > n16 = Sequence!(execute) -> (then_0, then_1, then_2:Exec, then_3:Exec, then_4:Exec, then_5:Exec, then_6:Exec)
   > n17 = GetFormat!(execute, self, tex:Texture Object Reference=Get Texture2d()) -> (
    then
    sRGB:bool
    texture render target format1:ETextureRenderTargetFormat Enum=RTF_R8
    Original Compression Settings:TextureCompressionSettings Enum
    Temp Compression for render:TextureCompressionSettings Enum
   )
   > Set srgb(srgb:bool=@n17) -> (Output_Get:bool)
   > Set orig(orig:TextureCompressionSettings Enum=@n17.Original Compression Settings) -> (Output_Get:TextureCompressionSettings Enum)
   > Set temp(temp:TextureCompressionSettings Enum=@n17.Temp Compression for render) -> (Output_Get:TextureCompressionSettings Enum)
   > Set RT format(RT format:ETextureRenderTargetFormat Enum=@n17.texture render target format1) -> (Output_Get:ETextureRenderTargetFormat Enum=RTF_R8)
   n18 = GetSizeX!(self:Texture 2D Object Reference=Get Texture2d()) -> (ReturnValue:int=0)
   > n19 = log2_int!(execute, self, A:int=@n18) -> (then, ReturnValue1:int=0)
   n20 = GetSizeY!(self:Texture 2D Object Reference=Get Texture2d()) -> (ReturnValue:int=0)
   > n21 = log2_int(A=@n20) -> (ReturnValue1=0)
   n22 = "Max (Integer)"(A=@n19, B=@n21)
   > Set MipNum(MipNum:int=@n22) -> (Output_Get:int=0)
   > n23 = PrintString(InString=Append(A="Total: ", B="To String (Integer)"(InInt=@n22)))
   n24 = Get temp() -> (temp:TextureCompressionSettings Enum)
   SetEditorProperty(execute=[@n23.then, @n39.then], Object=@n8, PropertyName=CompressionSettings, PropertyValue:TextureCompressionSettings Enum=@n24)
   n25 = Get MipNum() -> (MipNum:int=0)
   n26 = For Loop(execute=@n16.then_1, LastIndex=@n25)
   n27 = Power!(Base:double=2, Exp:double="To Float (Integer)"(InInt=@n26)) -> (ReturnValue:double) {is=title K2Node_CallFunction}
   n28 = Ceil(A=float / float(A:int=@n20, B=@n27)) {is=title K2Node_CallFunction}
   n29 = Ceil(A=float / float(A=@n18, B=@n27))
   > n30 = CreateRenderTarget2D!(
    execute
    WorldContextObject:Object Reference=@n5
    Width:int=@n29
    Height:int=@n28
    Format:ETextureRenderTargetFormat Enum=Get RT format() -> (RT format:ETextureRenderTargetFormat Enum=RTF_R8)
    ClearColor:Linear Color Structure=(R=0,G=0,B=0,A=1)
    bAutoGenerateMipMaps:bool=false
    bSupportUAVs:bool=false
   ) -> (then, ReturnValue:Texture Render Target 2D Object Reference)
   n31 = Get srgb()
   > n32 = Set SRGB(SRGB=@n31, self=@n30)
   Set Array Elem!(
    execute=[@n38.then, @n32.then]
    TargetArray:Array of Texture Render Target 2D Object References=Get RTs()
    Index:int=@n26
    Item:"Texture Render Target 2D Object Reference (by ref)"=@n30
    bSizeToFit:bool=true
   ) -> (then)
   n33 = Make Int Vector 2!(X:int=@n29, Y:int=@n28) -> (IntVector2:Int Vector 2 Structure) {is=title K2Node_MakeStruct}
   > Set Array Elem(TargetArray:Array of Int Vector 2 Structures=Get resarray(), Index=@n26, Item:"Int Vector 2 Structure (by ref)"=@n33, bSizeToFit=true)
   > PrintString(InString="To String (IntVector2)"!(InIntVec2:Int Vector 2 Structure=@n33) -> (ReturnValue:String))
   n34 = Get rawMat() -> (rawMat:Material Instance Dynamic Object Reference)
   SetTextureParameterValue(execute=@n16.then_3, self=@n34, ParameterName=Tex, Value=Get Texture2d())
   > DrawMaterialToRenderTarget!(
    execute
    WorldContextObject:Object Reference=@n5
    TextureRenderTarget:Texture Render Target 2D Object Reference="Get (a copy)"(Array:Array of Texture Render Target 2D Object References=Get RTs()) -> (Output:Texture Render Target 2D Object Reference)
    Material:Material Interface Object Reference=@n34
   ) -> (then)
   n35 = For Loop(execute=@n16.then_4, FirstIndex=1, LastIndex=Get MipNum() -> (MipNum=0))
   > SetTextureParameterValue(
    self=Get MinMat() -> (MinMat:Material Instance Dynamic Object Reference)
    ParameterName=Tex
    Value="Get (a copy)"(Array=Get RTs(), Dimension 1=int - int(A=@n35))
   )
   > DrawMaterialToRenderTarget(WorldContextObject=@n5, TextureRenderTarget="Get (a copy)"(Array=Get RTs(), Dimension 1=@n35), Material=Get MinMat())
   CreateMaximumMipMapTexture() {is=title K2Node_FunctionEntry}
   > Clear(TargetArray:Array of Int Vector 2 Structures=Get resarray())
   > n36 = Clear(TargetArray:Array of Texture Render Target 2D Object References=Get RTs())
   GetEditorProperty!(execute:Exec, Object:Object Reference=@n6.ReturnValue, PropertyName:Name=LossyCompressionAmount) -> (then:Exec, PropertyValue:Wildcard, ReturnValue:bool)
   > PrintString(InString=Hello)
   n37 = Reroute Node()
   n38 = SetEditorProperty(Object=@n30, PropertyName=SRGB, PropertyValue:bool=@n31)
   n39 = Set CompressionSettings(CompressionSettings=@n24, self=Get Texture2d())
   "To String (Float)"(InDouble=@n27)
   MakeLiteralString(Value=&/T_QuadTree)
   Append!(A="To String (Integer)"(InInt=@n29), B=" ", C:String="To String (Integer)"(InInt=@n28)) -> (ReturnValue)
   int + int(A=@n25)
   Get Texture2d()
   Get orig() -> (orig:TextureCompressionSettings Enum)
   Get RTs()
   Set CompressionSettings(CompressionSettings=Get orig())
   PrintString(execute=@n6.then, InString=@n4)
   > Branch(Condition=@n6)
   > n40 = PrintString(InString=@n6.OutErrorMessage)
  ]}
 EventGraph {links [], nodes []}
 GetFormat {nodes [
   n1 = GetFormat!() -> (then, tex:Texture Object Reference) {is=title K2Node_FunctionEntry}
   > n2 = Sequence!(execute) -> (then_0, then_1)
   n3 = Get CompressionSettings(self=@n1)
   > n4 = Branch(Condition=OR Boolean!(
     A="Equal (Enum)"(A:TextureCompressionSettings Enum=@n3, B:TextureCompressionSettings Enum=TC_Default)
     B="Equal (Enum)"(A=@n3, B=TC_Masks)
     C:bool="Equal (Enum)"(A=@n3, B=TC_VectorDisplacementmap)
     D:bool="Equal (Enum)"(A=@n3, B=TC_EditorIcon)
    ) -> (ReturnValue))
   n5 = Get SRGB!(self:Texture Object Reference=@n1) -> (SRGB:bool)
   > n6 = Branch(Condition=@n5)
   > n7 = Set texture render target format(texture render target format:ETextureRenderTargetFormat Enum=RTF_RGBA8_SRGB) -> (Output_Get:ETextureRenderTargetFormat Enum=RTF_R8)
   Set Temp Comp for render(execute=[@n8.then, @n7.then], Temp Comp for render:TextureCompressionSettings Enum=TC_EditorIcon) -> (Output_Get:TextureCompressionSettings Enum)
   n8 = Set texture render target format(execute=@n6.else, texture render target format=RTF_RGBA8) -> (Output_Get=RTF_R8)
   n9 = Branch(execute=@n4.else, Condition=OR Boolean!(A="Equal (Enum)"(A=@n3, B=TC_HDR), B="Equal (Enum)"(A=@n3, B=TC_HDR_Compressed)) -> (ReturnValue))
   > Set texture render target format(texture render target format=RTF_RGBA16f) -> (Output_Get=RTF_R8)
   > Set Temp Comp for render(Temp Comp for render=TC_HDR)
   Branch(execute=@n9.else, Condition="Equal (Enum)"(A=@n3, B=TC_HDR_F32))
   > Set texture render target format(texture render target format=RTF_RGBA32f) -> (Output_Get=RTF_R8)
   > Set Temp Comp for render(Temp Comp for render=TC_HDR_F32)
   Return Node!(
    execute=@n2.then_1
    sRGB:bool=@n5
    texture render target format1:ETextureRenderTargetFormat Enum=Get texture render target format() -> (texture render target format:ETextureRenderTargetFormat Enum=RTF_R8)
    Original Compression Settings:TextureCompressionSettings Enum=@n3
    Temp Compression for render:TextureCompressionSettings Enum=Get Temp Comp for render() -> (Temp Comp for render:TextureCompressionSettings Enum)
   )
  ]}
 log2_int {nodes [
   n1 = log2_int!() -> (then, A:int) {is=title K2Node_FunctionEntry}
   n2 = Log!(A:double="To Float (Integer)"(InInt=@n1), Base:double=2) -> (ReturnValue:double)
   > Return Node!(execute, ReturnValue1:int=Ceil(A=@n2))
   Round(A=@n2) {is=title K2Node_CallFunction}
  ]}
]
variables [
 NonSqrImage {type=Texture2D, default=/Engine/EngineSky/T_Sky_Clouds_M}
 rawMat {type=MaterialInstanceDynamic}
 resarray {type="array<IntVector2>", default []}
 RTs {type="array<TextureRenderTarget2D>", default []}
 Texture2d {type=Texture2D}
]
