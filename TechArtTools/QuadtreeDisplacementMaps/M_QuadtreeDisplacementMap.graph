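# Two-sided lit material that renders height-map relief by quadtree displacement mapping, a parallax ray march through texture mips.
# A custom HLSL node marches the tangent-space view ray, moving between mips to skip empty cells, then interpolates the final intersection.
# That hit UV samples Diffuse and Normal with pre-march derivatives to keep mip selection stable.
# Additional nodes derive depth-offset distances and a Debug-switched step-count heatmap; engine parallax occlusion nodes are also present.
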Material /TechArtTools/QuadtreeDisplacementMaps/M_QuadtreeDisplacementMap
graph {
 expressions {
  MaterialExpressionAbs_0 {class=MaterialExpressionAbs, inputs [None {from=MaterialExpressionCustom_0, output=return}], properties {}}
  MaterialExpressionAbs_1 {class=MaterialExpressionAbs, inputs [None {from=MaterialExpressionDotProduct_1, output=""}], properties {}}
  MaterialExpressionAbs_13 {class=MaterialExpressionAbs, inputs [None {from=MaterialExpressionDDX_13, output=""}], properties {}}
  MaterialExpressionAbs_14 {class=MaterialExpressionAbs, inputs [None {from=MaterialExpressionDDY_13, output=""}], properties {}}
  MaterialExpressionAbs_3 {class=MaterialExpressionAbs, inputs [None {from=MaterialExpressionDotProduct_3, output=""}], properties {}}
  MaterialExpressionAdd_0 {
   class=MaterialExpressionAdd
   inputs [
    A {from=MaterialExpressionMultiply_9}
    B {from=MaterialExpressionMultiply_5}
   ]
  }
  MaterialExpressionAdd_1 {
   class=MaterialExpressionAdd
   inputs [
    A {from=MaterialExpressionMax_0}
    B {from=""}
   ]
  }
  MaterialExpressionAdd_2 {
   class=MaterialExpressionAdd
   inputs [
    A {from=MaterialExpressionMultiply_16}
    B {from=MaterialExpressionConvert_0}
   ]
  }
  MaterialExpressionAppendVector_0 {
   class=MaterialExpressionAppendVector
   inputs [
    A {from=MaterialExpressionComponentMask_4, output=""}
    B {from=MaterialExpressionMultiply_11, output=""}
   ]
  }
  MaterialExpressionAppendVector_1 {
   class=MaterialExpressionAppendVector
   inputs [
    A {from=MaterialExpressionSubtract_2, output=""}
    B {from=MaterialExpressionComponentMask_12, output=""}
   ]
  }
  MaterialExpressionAppendVector_2 {
   class=MaterialExpressionAppendVector
   inputs [
    A {from=MaterialExpressionMultiply_9, output=""}
    B {from=MaterialExpressionConstant_10, output=""}
   ]
  }
  MaterialExpressionCameraPositionWS_0 {class=MaterialExpressionCameraPositionWS, inputs [], properties {}}
  MaterialExpressionCameraVectorWS_0 {class=MaterialExpressionCameraVectorWS, inputs [], properties {}}
  MaterialExpressionCameraVectorWS_1 {class=MaterialExpressionCameraVectorWS, inputs [], properties {}}
  MaterialExpressionCameraVectorWS_2 {class=MaterialExpressionCameraVectorWS, inputs [], properties {}}
  MaterialExpressionCameraVectorWS_3 {class=MaterialExpressionCameraVectorWS, inputs [], properties {}}
  MaterialExpressionCameraVectorWS_9 {class=MaterialExpressionCameraVectorWS, inputs [], properties {}}
  MaterialExpressionComponentMask_0 {class=MaterialExpressionComponentMask, inputs [None {from=MaterialExpressionTextureProperty_0}], properties {r=true}}
  MaterialExpressionComponentMask_1 {class=MaterialExpressionComponentMask, inputs [None {from=MaterialExpressionTextureProperty_0}], properties {g=true}}
  MaterialExpressionComponentMask_10 {class=MaterialExpressionComponentMask, inputs [None {from=""}], properties {g=true, r=true}}
  MaterialExpressionComponentMask_11 {class=MaterialExpressionComponentMask, inputs [None {from=MaterialExpressionReroute_1}], properties {g=true, r=true}}
  MaterialExpressionComponentMask_12 {class=MaterialExpressionComponentMask, inputs [None {from=MaterialExpressionReroute_1}], properties {b=true}}
  MaterialExpressionComponentMask_13 {class=MaterialExpressionComponentMask, inputs [None {from=MaterialExpressionCustom_0, output=return}], properties {b=true, g=true, r=true}}
  MaterialExpressionComponentMask_14 {class=MaterialExpressionComponentMask, inputs [None {from=MaterialExpressionReroute_5}], properties {b=true}}
  MaterialExpressionComponentMask_2 {class=MaterialExpressionComponentMask, inputs [None {from=MaterialExpressionCustom_0, output=return}], properties {g=true, r=true}}
  MaterialExpressionComponentMask_3 {class=MaterialExpressionComponentMask, inputs [None {from=MaterialExpressionTextureCoordinate_0}], properties {r=true}}
  MaterialExpressionComponentMask_4 {class=MaterialExpressionComponentMask, inputs [None {from=MaterialExpressionAppendVector_1}], properties {g=true, r=true}}
  MaterialExpressionComponentMask_5 {class=MaterialExpressionComponentMask, inputs [None {from=MaterialExpressionAppendVector_1}], properties {b=true}}
  MaterialExpressionComponentMask_6 {class=MaterialExpressionComponentMask, inputs [None {from=MaterialExpressionCustom_0, output=return}], properties {g=true, r=true}}
  MaterialExpressionComponentMask_7 {class=MaterialExpressionComponentMask, inputs [None {from=MaterialExpressionCustom_0, output=return}], properties {b=true}}
  MaterialExpressionComponentMask_8 {class=MaterialExpressionComponentMask, inputs [None {from=MaterialExpressionTransform_0}], properties {g=true, r=true}}
  MaterialExpressionComponentMask_9 {class=MaterialExpressionComponentMask, inputs [None {from=MaterialExpressionTransform_0}], properties {b=true}}
  MaterialExpressionConstant2Vector_0 {class=MaterialExpressionConstant2Vector, properties {g=0.5, r=0.5}}
  MaterialExpressionConstant3Vector_0 {class=MaterialExpressionConstant3Vector, properties {constant=<rgba 0 1 0 1>}}
  MaterialExpressionConstant3Vector_1 {class=MaterialExpressionConstant3Vector, properties {constant=<rgba 1 0 0 1>}}
  MaterialExpressionConstant3Vector_2 {class=MaterialExpressionConstant3Vector, properties {constant=<rgba 1 1 -1 0>}}
  MaterialExpressionConstant3Vector_3 {class=MaterialExpressionConstant3Vector, properties {constant=<rgba 0 1 0 1>}}
  MaterialExpressionConstant3Vector_4 {class=MaterialExpressionConstant3Vector, properties {constant=<rgba 1 0 0 1>}}
  MaterialExpressionConstant3Vector_6 {class=MaterialExpressionConstant3Vector, properties {constant=<rgba 1 1 1 0>}}
  MaterialExpressionConstant4Vector_0 {class=MaterialExpressionConstant4Vector, inputs [], properties {constant=<rgba 1 0 0 0>}}
  MaterialExpressionConstant_0 {class=MaterialExpressionConstant, properties {r=0.125}}
  MaterialExpressionConstant_1 {class=MaterialExpressionConstant, properties {r=1}}
  MaterialExpressionConstant_10 {class=MaterialExpressionConstant}
  MaterialExpressionConstant_11 {class=MaterialExpressionConstant}
  MaterialExpressionConstant_14 {class=MaterialExpressionConstant}
  MaterialExpressionConstant_2 {class=MaterialExpressionConstant, properties {r=6}}
  MaterialExpressionConstant_3 {class=MaterialExpressionConstant, properties {r=64}}
  MaterialExpressionConstant_4 {class=MaterialExpressionConstant}
  MaterialExpressionConstant_5 {class=MaterialExpressionConstant, properties {r=0.5}}
  MaterialExpressionConstant_6 {class=MaterialExpressionConstant, properties {r=0.5}}
  MaterialExpressionConstant_7 {class=MaterialExpressionConstant, properties {r=2}}
  MaterialExpressionConstant_8 {class=MaterialExpressionConstant}
  MaterialExpressionConstant_9 {class=MaterialExpressionConstant}
  MaterialExpressionConvert_0 {
   class=MaterialExpressionConvert
   inputs [
    None {from=MaterialExpressionConstant_11}
    None {from=MaterialExpressionConstant_11}
    None {from=MaterialExpressionMultiply_20}
   ]
   properties {
    nodeName=Make Vector3
    convertInputs [
     {
      type=Scalar
      expressionInput {
       expression=&self:MaterialExpressionConstant_11
       inputName=None
       mask=0
       maskA=0
       maskB=0
       maskG=0
       maskR=0
       outputIndex=0
      }
     }
     {
      type=Scalar
      expressionInput {
       expression=&self:MaterialExpressionConstant_11
       inputName=None
       mask=0
       maskA=0
       maskB=0
       maskG=0
       maskR=0
       outputIndex=0
      }
     }
     {
      type=Scalar
      expressionInput {
       expression=&self:MaterialExpressionMultiply_20
       inputName=None
       mask=0
       maskA=0
       maskB=0
       maskG=0
       maskR=0
       outputIndex=0
      }
     }
    ]
    convertOutputs [{type=Vector3}]
   }
  }
  MaterialExpressionCustom_0 {
   class=MaterialExpressionCustom
   inputs [
    Tex {from=MaterialExpressionTextureObjectParameter_0, output=""}
    Ray {from=MaterialExpressionMultiply_4, output=""}
    uv {from=MaterialExpressionAdd_0, output=""}
    numMips {from=MaterialExpressionRound_0, output=""}
    maxSteps {from=MaterialExpressionScalarParameter_6, output=""}
    hr {from=MaterialExpressionScalarParameter_2, output=""}
   ]
   properties {
    code="\r\nint resolution = int(pow(2, numMips));\r\n\r\n//float delta = 1.0/(2.0*resolution);//0.00001;\r\nfloat delta = 0.00001;\r\n// from custom node input\r\n//float3 Ray;\r\n//float2 uv;\r\nfloat3 Origin = float3(uv, 0);\r\n//int numMips;\r\n//int MaxSteps;\r\n\r\nstruct func {\r\n    static bool advanceCursor(  float2 currentCursor,\r\n                        float2 startpoint,\r\n                        float3 ray,\r\n                        float depth,\r\n                        inout float t_cursor, \r\n                        float delta,\r\n                        bool texelCenter,\r\n                        int resolution){\r\n    \r\n        float2 ray2 = ray.xy;\r\n        float2 term1 = floor((currentCursor * resolution));// + ((texelCenter)*0.5));\r\n        if(texelCenter){\r\n\r\n            term1 += 0.5;\r\n        }\r\n        //float2 term2 =  (0.5 * sign(ray2)) + ((1.0-(texelCenter))*0.5);\r\n        float2 term2 =  (0.5 * sign(ray2));\r\n        if(!texelCenter){\r\n           term2 += 0.5;\r\n        }\r\n\r\n        float2 predictedPos = (((term1 + term2)/resolution)-startpoint) / ray2;\r\n\r\n        float t = (depth/ray.z);\r\n\r\n        float t_min = min(predictedPos.x, predictedPos.y);\r\n        \r\n        //output with delta in cursor in 1st term anyway because it gets subtracted after\r\n        t_cursor = max(t_cursor+delta, min(t, t_min+delta));\r\n        \r\n        // this makes it do a full pixel in last level to ensure one full crossing \r\n        if(texelCenter){\r\n            t_cursor = t_min+delta;\r\n        }\r\n        //t_cursor = lerp(t_cursor,  t_min+delta, texelCenter);\r\n        \r\n        return  (t < (t_min+delta)) ;\r\n    }\r\n};\r\n\r\nfunc f;\r\n\r\nfloat2 startpoint = uv;\r\nfloat t_cursor = 0;\r\n\r\nfloat3 cursor = float3(uv, 0);\r\nfloat3 cursorSample = cursor;\r\n\r\n\r\n\r\n\r\n\r\n\r\nfloat t = 0;\r\nfloat t2d = 0;\r\nfloat t2dSqr = 0;\r\n\r\nint i = 0;\r\nint res;\r\nint bHitInsideCell;\r\n\r\nfloat t_min = 0;\r\nint finish = 1;\r\nint steps = 0;\r\n\r\n\r\n\r\nint currentlevel = numMips;\r\nint maxLevel = currentlevel;\r\n\r\nfloat hitpos;\r\n\r\n[loop]\r\nwhile ((currentlevel >= 0) &&&& (i < maxSteps)) {\r\n\r\n\r\n    //float tm = 0;\r\n    resolution = int(pow(2, numMips-currentlevel));\r\n    //delta = 1.0/(2*resolution);\r\n    //int resolution = 2 << int(numMips-currentlevel);\r\n\r\n    bool isLastLevel = (currentlevel == 0);\r\n\r\n\r\n\r\n\r\n    // make sure we texel centre the forward offset sample pos (I forgot to do this until the end!)\r\n    // need this to adapt to non square images (array of vector2 for res?)\r\n    float2 samplePos = cursorSample.xy;\r\n    samplePos *= resolution ;\r\n    samplePos = floor(samplePos) + 0.5f;\r\n    samplePos /=  resolution ;\r\n\r\n    // use sample centered on higher lods and boundary cursor on last\r\n    float2 samplePosFin;\r\n    if(isLastLevel){\r\n        samplePosFin  = cursor.xy;\r\n    } else {\r\n        samplePosFin  = samplePos;\r\n    }\r\n    //samplePosFin = frac(samplePosFin);\r\n    //samplePosFin = lerp(samplePos, cursor.xy, currentlevel == 0);\r\n\r\n    hitpos = -(1.0-Tex.SampleLevel(TexSampler,samplePosFin, currentlevel).r)*hr;\r\n\r\n    i += 1;\r\n\r\n    //float3 testCursor = Origin + Ray * max((t_cursor-delta),0);\r\n    if (((cursor.z < hitpos)) &&&& (isLastLevel)){\r\n        \r\n        break;\r\n    }\r\n\r\n    \r\n\r\n    bHitInsideCell = f.advanceCursor(cursorSample.xy, startpoint, Ray, hitpos, t_cursor, delta, isLastLevel, resolution);\r\n    \r\n    \r\n\r\n\r\n    \r\n    // snap cursor to boundary by subtracting delta but for next sample, keep it ahead\r\n    //cursor = testCursor;\r\n\r\n    // dont max here it causes artefacts (wierd, putting t_cursor-delta in brackets gives artefact\r\n    cursor =  Origin + Ray * (t_cursor-delta);\r\n\r\n    // but can max here to prevent going back \r\n    cursorSample = Origin + Ray * t_cursor;\r\n\r\n\r\n    \r\n\r\n\r\n    // move up if we hit a boundary and down if we are inside a cell\r\n    //currentlevel -= (int(bHitInsideCell)*2)-1;\r\n    if (bHitInsideCell)\r\n    {\r\n        // Possible hit in this cell: descend toward finer mips.\r\n        currentlevel -= 1;\r\n    }\r\n    else\r\n    {\r\n        // Empty cell crossed: go coarser again so we can skip faster.\r\n        currentlevel += 1;\r\n    }\r\n\r\n    // only ever go back up a certain number of levels. (-1 or -2 seems to be best) .\r\n    // just an experiment which I found was faster due to not changing levels so much. TODO, have a countdown\r\n    maxLevel = min(maxLevel, currentlevel+3);\r\n    currentlevel = min(currentlevel, maxLevel);\r\n    \r\n\r\n    //prevent level from going deeper than it can\r\n    currentlevel = max(currentlevel, 0);\r\n\r\n}\r\n\r\n// float tm;\r\n// //start back a little bit in case we are already exactly on the boundary\r\n// t2 = advanceCursor(set(v@cursor.x, v@cursor.z)-set(v@ray.x, v@ray.z)*0.001, set(v@cursor.x, v@cursor.z),  -v@ray, hitpos.y, t_cursor, tm, delta,1, i@resol);\r\n// v@oldcursor = v@cursor - v@ray* (t_cursor-delta);\r\n// setpointattrib(0, \"P\", curOld, v@oldcursor); \r\n// vector prevhit;\r\n\r\n// res = intersect(1, s@groupLevel,  set(v@oldcursor.x, 0.0, v@oldcursor.z), v@N, prevhit, uvw);\r\n\r\n// float heightDiffSqrd = abs(v@oldcursor.y-v@cursor.y);\r\n// heightDiffSqrd *= heightDiffSqrd;\r\n\r\n// float a = abs(prevhit.y - v@oldcursor.y);\r\n// float b = abs(hitpos.y - v@cursor.y);\r\n// vector intersect = lerp(prevhit, hitpos, a/(a+b));\r\n\r\n//go back to prev boundary\r\n\r\nresolution = int(pow(2, numMips));\r\nf.advanceCursor(cursor.xy - (normalize(Ray.xy) * (1.0/(resolution))), cursor.xy, -Ray, 1, t_cursor, delta, true, resolution);\r\nfloat3 backCursor = cursor - Ray * t_cursor-delta;\r\nfloat backHitpos = -(1.0-Tex.SampleLevel(TexSampler, backCursor.xy, 0).r)*hr;\r\n\r\nfloat a = abs(backHitpos - backCursor.z);\r\nfloat b = abs(hitpos - cursor.z);\r\nfloat3 intersect = lerp(backCursor, cursor , (a/(a+b)));\r\n\r\n//intersect = float3(intersect.x, intersect.y, 1.0-(abs(intersect.z))/hr);\r\n\r\nstepsTaken = i;\r\nfloat4 output = float4(intersect, 1);\r\nreturn output;\r\n\r\n\r\n\r\n\r\n\r\n"
    showCode=true
    additionalOutputs [{outputName=stepsTaken, outputType=CMOT_Float1}]
    inputs [
     {
      inputName=Tex
      input {
       expression=&self:MaterialExpressionTextureObjectParameter_0
       inputName=None
       mask=0
       maskA=0
       maskB=0
       maskG=0
       maskR=0
       outputIndex=0
      }
     }
     {
      inputName=Ray
      input {
       expression=&self:MaterialExpressionMultiply_4
       inputName=None
       mask=0
       maskA=0
       maskB=0
       maskG=0
       maskR=0
       outputIndex=0
      }
     }
     {
      inputName=uv
      input {
       expression=&self:MaterialExpressionAdd_0
       inputName=None
       mask=0
       maskA=0
       maskB=0
       maskG=0
       maskR=0
       outputIndex=0
      }
     }
     {
      inputName=numMips
      input {
       expression=&self:MaterialExpressionRound_0
       inputName=None
       mask=0
       maskA=0
       maskB=0
       maskG=0
       maskR=0
       outputIndex=0
      }
     }
     {
      inputName=maxSteps
      input {
       expression=&self:MaterialExpressionScalarParameter_6
       inputName=None
       mask=0
       maskA=0
       maskB=0
       maskG=0
       maskR=0
       outputIndex=0
      }
     }
     {
      inputName=hr
      input {
       expression=&self:MaterialExpressionScalarParameter_2
       inputName=None
       mask=0
       maskA=0
       maskB=0
       maskG=0
       maskR=0
       outputIndex=0
      }
     }
    ]
   }
  }
  MaterialExpressionDDX_0 {class=MaterialExpressionDDX, inputs [Value {from=MaterialExpressionAdd_0, output=""}], properties {}}
  MaterialExpressionDDX_13 {class=MaterialExpressionDDX, inputs [Value {from=MaterialExpressionReroute_3, output=""}], properties {}}
  MaterialExpressionDDX_14 {class=MaterialExpressionDDX, inputs [Value {from=MaterialExpressionWorldPosition_4, output=XYZ}], properties {}}
  MaterialExpressionDDY_0 {class=MaterialExpressionDDY, inputs [Value {from=MaterialExpressionAdd_0, output=""}], properties {}}
  MaterialExpressionDDY_13 {class=MaterialExpressionDDY, inputs [Value {from=MaterialExpressionReroute_3, output=""}], properties {}}
  MaterialExpressionDDY_14 {class=MaterialExpressionDDY, inputs [Value {from=MaterialExpressionWorldPosition_4, output=XYZ}], properties {}}
  MaterialExpressionDivide_0 {
   class=MaterialExpressionDivide
   inputs [
    A {from=MaterialExpressionCustom_0, output=stepsTaken}
    B {from=MaterialExpressionConstant_3}
   ]
   properties {constB=200}
  }
  MaterialExpressionDivide_1 {
   class=MaterialExpressionDivide
   inputs [
    A {from=MaterialExpressionLength_0}
    B {from=MaterialExpressionDivide_39}
   ]
  }
  MaterialExpressionDivide_18 {
   class=MaterialExpressionDivide
   inputs [
    A {from=MaterialExpressionMaterialFunctionCall_22, output=V2 Length}
    B {from=MaterialExpressionMaterialFunctionCall_20, output=V3 Length}
   ]
  }
  MaterialExpressionDivide_19 {
   class=MaterialExpressionDivide
   inputs [
    A {from=MaterialExpressionMaterialFunctionCall_23, output=V2 Length}
    B {from=MaterialExpressionMaterialFunctionCall_21, output=V3 Length}
   ]
  }
  MaterialExpressionDivide_2 {
   class=MaterialExpressionDivide
   inputs [
    A {from=MaterialExpressionLength_1}
    B {from=MaterialExpressionAbs_1}
   ]
  }
  MaterialExpressionDivide_39 {
   class=MaterialExpressionDivide
   inputs [
    A {from=MaterialExpressionMax_4}
    B {from=MaterialExpressionAbs_3}
   ]
  }
  MaterialExpressionDivide_9 {
   class=MaterialExpressionDivide
   inputs [
    A {from=MaterialExpressionMultiply_17}
    B {from=MaterialExpressionComponentMask_9}
   ]
   properties {constB=1}
  }
  MaterialExpressionDotProduct_0 {
   class=MaterialExpressionDotProduct
   inputs [
    A {from=MaterialExpressionCameraVectorWS_1, output=""}
    B {from=MaterialExpressionVertexNormalWS_0, output=""}
   ]
   properties {}
  }
  MaterialExpressionDotProduct_1 {
   class=MaterialExpressionDotProduct
   inputs [
    A {from=MaterialExpressionMaterialFunctionCall_10, output=Result}
    B {from=MaterialExpressionCameraVectorWS_3, output=""}
   ]
   properties {}
  }
  MaterialExpressionDotProduct_3 {
   class=MaterialExpressionDotProduct
   inputs [
    A {from=MaterialExpressionMaterialFunctionCall_9, output=Result}
    B {from=MaterialExpressionCameraVectorWS_2, output=""}
   ]
   properties {}
  }
  MaterialExpressionFrac_0 {class=MaterialExpressionFrac, inputs [None {from=MaterialExpressionMaterialFunctionCall_0, output=Pixel Depth Offset}], properties {}}
  MaterialExpressionFrac_1 {class=MaterialExpressionFrac, inputs [None {from=MaterialExpressionMultiply_13, output=""}], properties {}}
  MaterialExpressionLength_0 {class=MaterialExpressionLength, inputs [None {from=MaterialExpressionAppendVector_0, output=""}], properties {}}
  MaterialExpressionLength_1 {class=MaterialExpressionLength, inputs [None {from=MaterialExpressionTransform_2, output=""}], properties {}}
  MaterialExpressionLinearInterpolate_0 {
   class=MaterialExpressionLinearInterpolate
   inputs [
    A {from=MaterialExpressionConstant3Vector_0}
    B {from=MaterialExpressionConstant3Vector_1}
    Alpha {from=MaterialExpressionSaturate_0}
   ]
  }
  MaterialExpressionLinearInterpolate_1 {
   class=MaterialExpressionLinearInterpolate
   inputs [
    A {from=MaterialExpressionConstant3Vector_3}
    B {from=MaterialExpressionConstant3Vector_4}
    Alpha {from=MaterialExpressionComponentMask_3}
   ]
  }
  MaterialExpressionLinearInterpolate_2 {
   class=MaterialExpressionLinearInterpolate
   inputs [
    A {from=MaterialExpressionScalarParameter_7}
    B {from=MaterialExpressionScalarParameter_6}
    Alpha {from=MaterialExpressionOneMinus_3}
   ]
  }
  MaterialExpressionLinearInterpolate_3 {
   class=MaterialExpressionLinearInterpolate
   inputs [
    A {from=MaterialExpressionScalarParameter_8}
    B {from=MaterialExpressionScalarParameter_9}
    Alpha {from=MaterialExpressionScalarBlueNoise_0}
   ]
  }
  MaterialExpressionLogarithm2_0 {class=MaterialExpressionLogarithm2, inputs [X {from=MaterialExpressionComponentMask_0, output=""}], properties {}}
  MaterialExpressionLogarithm2_1 {class=MaterialExpressionLogarithm2, inputs [X {from=MaterialExpressionComponentMask_1, output=""}], properties {}}
  MaterialExpressionMaterialFunctionCall_0 {
   class=MaterialExpressionMaterialFunctionCall
   inputs [
    Heightmap Texture {from=MaterialExpressionTextureObjectParameter_0}
    Height Ratio {from=MaterialExpressionScalarParameter_0}
    Min Steps {from=""}
    Max Steps {from=""}
    UVs {from=""}
    Heightmap Channel {from=MaterialExpressionConstant4Vector_0, output=RGBA}
    Reference Plane {from=MaterialExpressionScalarParameter_3}
    ----------------------------- {from=""}
    Use World Coordinates {from=MaterialExpressionStaticBool_1}
    Specify Manual Texture Size {from=""}
    Manual Texture Size {from=""}
    ------------------------------ {from=""}
    "Render Shadows (Occlusion Mapping)" {from=""}
    Light Vector {from=""}
    Shadow Steps {from=""}
    Shadow Penumbra {from=""}
    Transform To VertexNormal {from=""}
   ]
   properties {materialFunction=/Engine/Functions/Engine_MaterialFunctions01/Texturing/ParallaxOcclusionMapping}
  }
  MaterialExpressionMaterialFunctionCall_1 {
   class=MaterialExpressionMaterialFunctionCall
   inputs [
    Input {from=MaterialExpressionTextureSampleParameter2D_1, output=RGB}
    Input Low {from=MaterialExpressionConstant_0}
    Input High {from=MaterialExpressionConstant_1}
   ]
   properties {materialFunction=&RemapValue_0-1}
  }
  MaterialExpressionMaterialFunctionCall_10 {
   class=MaterialExpressionMaterialFunctionCall
   inputs []
   properties {materialFunction=&CameraDirectionVector}
  }
  MaterialExpressionMaterialFunctionCall_2 {
   class=MaterialExpressionMaterialFunctionCall
   inputs [
    In {from=""}
    Contrast {from=MaterialExpressionConstant_7}
   ]
   properties {materialFunction=&CheapContrast}
  }
  MaterialExpressionMaterialFunctionCall_20 {
   class=MaterialExpressionMaterialFunctionCall
   inputs [
    Vector 3 {from=MaterialExpressionDDX_14}
    Vector 2 {from=""}
   ]
   properties {materialFunction=/Engine/Functions/Engine_MaterialFunctions02/Utility/VectorLength}
  }
  MaterialExpressionMaterialFunctionCall_21 {
   class=MaterialExpressionMaterialFunctionCall
   inputs [
    Vector 3 {from=MaterialExpressionDDY_14}
    Vector 2 {from=""}
   ]
   properties {materialFunction=&VectorLength}
  }
  MaterialExpressionMaterialFunctionCall_22 {
   class=MaterialExpressionMaterialFunctionCall
   inputs [
    Vector 3 {from=""}
    Vector 2 {from=MaterialExpressionAbs_13}
   ]
   properties {materialFunction=&VectorLength}
  }
  MaterialExpressionMaterialFunctionCall_23 {
   class=MaterialExpressionMaterialFunctionCall
   inputs [
    Vector 3 {from=""}
    Vector 2 {from=MaterialExpressionAbs_14}
   ]
   properties {materialFunction=&VectorLength}
  }
  MaterialExpressionMaterialFunctionCall_3 {
   class=MaterialExpressionMaterialFunctionCall
   inputs [
    Number {from=MaterialExpressionRound_0}
    MaximumNumberOfDigits {from=""}
    UVs {from=""}
    DebugTextLocation RG_UpperRight BA_LowerLeft {from=""}
   ]
   properties {materialFunction=&DebugScalarValues}
  }
  MaterialExpressionMaterialFunctionCall_4 {
   class=MaterialExpressionMaterialFunctionCall
   inputs [
    Number {from=MaterialExpressionCustom_0, output=stepsTaken}
    MaximumNumberOfDigits {from=""}
    UVs {from=""}
    DebugTextLocation RG_UpperRight BA_LowerLeft {from=""}
   ]
   properties {materialFunction=&DebugScalarValues}
  }
  MaterialExpressionMaterialFunctionCall_8 {
   class=MaterialExpressionMaterialFunctionCall
   inputs [
    Heightmap Texture {from=""}
    Height Ratio {from=""}
    Min Steps {from=""}
    Max Steps {from=""}
    UVs {from=""}
    Heightmap Channel {from=""}
    Reference Plane {from=""}
    ----------------------------- {from=""}
    Use World Coordinates {from=""}
    Specify Manual Texture Size {from=""}
    Manual Texture Size {from=""}
    ------------------------------ {from=""}
    "Render Shadows (Occlusion Mapping)" {from=""}
    Light Vector {from=""}
    Shadow Steps {from=""}
    Shadow Penumbra {from=""}
    Transform To VertexNormal {from=""}
   ]
   properties {materialFunction=&ParallaxOcclusionMapping}
  }
  MaterialExpressionMaterialFunctionCall_9 {
   class=MaterialExpressionMaterialFunctionCall
   inputs []
   properties {materialFunction=&CameraDirectionVector}
  }
  MaterialExpressionMax_0 {
   class=MaterialExpressionMax
   inputs [
    A {from=MaterialExpressionLogarithm2_0}
    B {from=MaterialExpressionLogarithm2_1}
   ]
  }
  MaterialExpressionMax_4 {
   class=MaterialExpressionMax
   inputs [
    A {from=MaterialExpressionDivide_18}
    B {from=MaterialExpressionDivide_19}
   ]
  }
  MaterialExpressionMultiply_0 {
   class=MaterialExpressionMultiply
   inputs [
    A {from=MaterialExpressionNormalize_0}
    B {from=MaterialExpressionMultiply_4}
   ]
  }
  MaterialExpressionMultiply_1 {
   class=MaterialExpressionMultiply
   inputs [
    A {from=""}
    B {from=""}
   ]
   properties {constB=8}
  }
  MaterialExpressionMultiply_10 {
   class=MaterialExpressionMultiply
   inputs [
    A {from=MaterialExpressionTextureSampleParameter2D_1, output=RGB}
    B {from=MaterialExpressionScalarParameter_5}
   ]
  }
  MaterialExpressionMultiply_11 {
   class=MaterialExpressionMultiply
   inputs [
    A {from=MaterialExpressionOneMinus_4}
    B {from=MaterialExpressionReroute_0}
   ]
  }
  MaterialExpressionMultiply_12 {
   class=MaterialExpressionMultiply
   inputs [
    A {from=MaterialExpressionMaterialFunctionCall_0, output=Pixel Depth Offset}
    B {from=""}
   ]
   properties {constB=0.02}
  }
  MaterialExpressionMultiply_13 {
   class=MaterialExpressionMultiply
   inputs [
    A {from=MaterialExpressionMultiply_18}
    B {from=MaterialExpressionScalarParameter_12}
   ]
  }
  MaterialExpressionMultiply_14 {
   class=MaterialExpressionMultiply
   inputs [
    A {from=MaterialExpressionTextureSampleParameter2D_1, output=RGB}
    B {from=""}
   ]
  }
  MaterialExpressionMultiply_16 {
   class=MaterialExpressionMultiply
   inputs [
    A {from=MaterialExpressionComponentMask_13}
    B {from=MaterialExpressionConstant3Vector_6}
   ]
  }
  MaterialExpressionMultiply_17 {
   class=MaterialExpressionMultiply
   inputs [
    A {from=MaterialExpressionComponentMask_8}
    B {from=""}
   ]
   properties {constB=-1}
  }
  MaterialExpressionMultiply_18 {
   class=MaterialExpressionMultiply
   inputs [
    A {from=MaterialExpressionLength_1}
    B {from=MaterialExpressionAbs_1}
   ]
  }
  MaterialExpressionMultiply_19 {
   class=MaterialExpressionMultiply
   inputs [
    A {from=MaterialExpressionScalarParameter_2}
    B {from=MaterialExpressionDivide_9}
   ]
  }
  MaterialExpressionMultiply_2 {
   class=MaterialExpressionMultiply
   inputs [
    A {from=MaterialExpressionDDX_0}
    B {from=MaterialExpressionScalarParameter_10}
   ]
  }
  MaterialExpressionMultiply_20 {
   class=MaterialExpressionMultiply
   inputs [
    A {from=MaterialExpressionOneMinus_1}
    B {from=MaterialExpressionScalarParameter_2}
   ]
  }
  MaterialExpressionMultiply_3 {
   class=MaterialExpressionMultiply
   inputs [
    A {from=MaterialExpressionDDY_0}
    B {from=MaterialExpressionScalarParameter_10}
   ]
  }
  MaterialExpressionMultiply_4 {
   class=MaterialExpressionMultiply
   inputs [
    A {from=MaterialExpressionTransform_0}
    B {from=""}
   ]
   properties {constB=-1}
  }
  MaterialExpressionMultiply_5 {
   class=MaterialExpressionMultiply
   inputs [
    A {from=MaterialExpressionMultiply_19}
    B {from=MaterialExpressionMultiply_6}
   ]
  }
  MaterialExpressionMultiply_6 {
   class=MaterialExpressionMultiply
   inputs [
    A {from=MaterialExpressionOneMinus_1}
    B {from=""}
   ]
   properties {constB=-1}
  }
  MaterialExpressionMultiply_7 {
   class=MaterialExpressionMultiply
   inputs [
    A {from=MaterialExpressionTextureSampleParameter2D_1, output=R}
    B {from=MaterialExpressionLinearInterpolate_0}
   ]
  }
  MaterialExpressionMultiply_8 {
   class=MaterialExpressionMultiply
   inputs [
    A {from=MaterialExpressionNormalize_0}
    B {from=MaterialExpressionConstant3Vector_2}
   ]
  }
  MaterialExpressionMultiply_9 {
   class=MaterialExpressionMultiply
   inputs [
    A {from=MaterialExpressionTextureCoordinate_8}
    B {from=MaterialExpressionScalarParameter_4}
   ]
  }
  MaterialExpressionNormalize_0 {class=MaterialExpressionNormalize, inputs [VectorInput {from=MaterialExpressionMultiply_4, output=""}]}
  MaterialExpressionNormalize_1 {class=MaterialExpressionNormalize, inputs [VectorInput {from=MaterialExpressionSubtract_0, output=""}]}
  MaterialExpressionOneMinus_0 {class=MaterialExpressionOneMinus, inputs [None {from=MaterialExpressionSaturate_1}]}
  MaterialExpressionOneMinus_1 {class=MaterialExpressionOneMinus, inputs [None {from=MaterialExpressionScalarParameter_3}]}
  MaterialExpressionOneMinus_2 {class=MaterialExpressionOneMinus, inputs [None {from=MaterialExpressionMaterialFunctionCall_2, output=Result}]}
  MaterialExpressionOneMinus_3 {class=MaterialExpressionOneMinus, inputs [None {from=MaterialExpressionSaturate_2}]}
  MaterialExpressionOneMinus_4 {class=MaterialExpressionOneMinus, inputs [None {from=MaterialExpressionComponentMask_5}]}
  MaterialExpressionReroute_0 {class=MaterialExpressionReroute, inputs [None {from=MaterialExpressionMultiply_20}]}
  MaterialExpressionReroute_1 {class=MaterialExpressionReroute, inputs [None {from=MaterialExpressionCustom_0, output=return}]}
  MaterialExpressionReroute_2 {class=MaterialExpressionReroute, inputs [None {from=MaterialExpressionDivide_1}]}
  MaterialExpressionReroute_3 {class=MaterialExpressionReroute, inputs [None {from=MaterialExpressionMultiply_9}]}
  MaterialExpressionReroute_4 {class=MaterialExpressionReroute, inputs [None {from=MaterialExpressionAdd_2}]}
  MaterialExpressionReroute_5 {class=MaterialExpressionReroute, inputs [None {from=MaterialExpressionMultiply_5}]}
  MaterialExpressionRound_0 {class=MaterialExpressionRound, inputs [None {from=MaterialExpressionMax_0, output=""}], properties {}}
  MaterialExpressionSaturate_0 {class=MaterialExpressionSaturate, inputs [None {from=MaterialExpressionDivide_0}]}
  MaterialExpressionSaturate_1 {class=MaterialExpressionSaturate, inputs [None {from=MaterialExpressionMultiply_1}]}
  MaterialExpressionSaturate_2 {class=MaterialExpressionSaturate, inputs [None {from=MaterialExpressionDotProduct_0}]}
  MaterialExpressionScalarBlueNoise_0 {class=MaterialExpressionScalarBlueNoise, inputs [], properties {}}
  MaterialExpressionScalarParameter_0 {class=MaterialExpressionScalarParameter, properties {defaultValue=0.175, parameterName=hr}}
  MaterialExpressionScalarParameter_1 {class=MaterialExpressionScalarParameter, properties {defaultValue=64, parameterName=MaxSteps}}
  MaterialExpressionScalarParameter_10 {class=MaterialExpressionScalarParameter, properties {defaultValue=1, parameterName=m}}
  MaterialExpressionScalarParameter_11 {class=MaterialExpressionScalarParameter, properties {parameterName=Bias}}
  MaterialExpressionScalarParameter_12 {class=MaterialExpressionScalarParameter, properties {defaultValue=100, parameterName=si}}
  MaterialExpressionScalarParameter_2 {class=MaterialExpressionScalarParameter, properties {defaultValue=0.175, parameterName=hr}}
  MaterialExpressionScalarParameter_3 {class=MaterialExpressionScalarParameter, properties {defaultValue=1, parameterName=Refplane}}
  MaterialExpressionScalarParameter_4 {class=MaterialExpressionScalarParameter, properties {defaultValue=1, parameterName=uvmult}}
  MaterialExpressionScalarParameter_5 {class=MaterialExpressionScalarParameter, properties {defaultValue=2, parameterName=gain}}
  MaterialExpressionScalarParameter_6 {class=MaterialExpressionScalarParameter, properties {defaultValue=64, parameterName=MaxSteps}}
  MaterialExpressionScalarParameter_7 {class=MaterialExpressionScalarParameter, properties {defaultValue=24, parameterName=MinSteps}}
  MaterialExpressionScalarParameter_8 {class=MaterialExpressionScalarParameter, properties {defaultValue=0.0001, parameterName=low}}
  MaterialExpressionScalarParameter_9 {class=MaterialExpressionScalarParameter, properties {defaultValue=0.001, parameterName=high}}
  MaterialExpressionStaticBool_0 {class=MaterialExpressionStaticBool}
  MaterialExpressionStaticBool_1 {class=MaterialExpressionStaticBool}
  MaterialExpressionStaticSwitchParameter_0 {
   class=MaterialExpressionStaticSwitchParameter
   inputs [
    True {from=MaterialExpressionConstant_9, output=""}
    False {from=MaterialExpressionTextureSampleParameter2D_1, output=RGB}
   ]
   properties {parameterName=Debug}
  }
  MaterialExpressionStaticSwitchParameter_1 {
   class=MaterialExpressionStaticSwitchParameter
   inputs [
    True {from=MaterialExpressionLinearInterpolate_0, output=""}
    False {from=MaterialExpressionConstant_9, output=""}
   ]
   properties {parameterName=Debug}
  }
  MaterialExpressionSubtract_0 {
   class=MaterialExpressionSubtract
   inputs [
    A {from=MaterialExpressionWorldPosition_0, output=XYZ}
    B {from=MaterialExpressionCameraPositionWS_0}
   ]
  }
  MaterialExpressionSubtract_1 {
   class=MaterialExpressionSubtract
   inputs [
    A {from=MaterialExpressionRound_0}
    B {from=MaterialExpressionScalarParameter_11}
   ]
  }
  MaterialExpressionSubtract_2 {
   class=MaterialExpressionSubtract
   inputs [
    A {from=MaterialExpressionAdd_0}
    B {from=MaterialExpressionComponentMask_11}
   ]
  }
  MaterialExpressionSubtract_3 {
   class=MaterialExpressionSubtract
   inputs [
    A {from=MaterialExpressionAppendVector_2}
    B {from=MaterialExpressionReroute_4}
   ]
  }
  MaterialExpressionSubtract_4 {
   class=MaterialExpressionSubtract
   inputs [
    A {from=MaterialExpressionMultiply_16}
    B {from=MaterialExpressionConvert_0}
   ]
  }
  MaterialExpressionTextureCoordinate_0 {class=MaterialExpressionTextureCoordinate}
  MaterialExpressionTextureCoordinate_8 {class=MaterialExpressionTextureCoordinate}
  MaterialExpressionTextureObjectParameter_0 {
   class=MaterialExpressionTextureObjectParameter
   inputs [
    UVs {from="", output=""}
    Apply View MipBias {from=MaterialExpressionStaticBool_0, output=""}
   ]
   properties {parameterName=DIsp, samplerType=SAMPLERTYPE_LinearColor, texture=&/tile1_relief_H_QuadtreeDisplacement}
  }
  MaterialExpressionTextureProperty_0 {class=MaterialExpressionTextureProperty, inputs [Tex {from=MaterialExpressionTextureObjectParameter_0, output=""}], properties {}}
  MaterialExpressionTextureSampleParameter2D_0 {
   class=MaterialExpressionTextureSampleParameter2D
   inputs [
    UVs {from=MaterialExpressionComponentMask_2, output=""}
    "DDX(UVs)" {from=MaterialExpressionDDX_0, output=""}
    "DDY(UVs)" {from=MaterialExpressionDDY_0, output=""}
    Apply View MipBias {from="", output=""}
   ]
   properties {mipValueMode=TMVM_Derivative, parameterName=Normal, samplerType=SAMPLERTYPE_Normal, texture=&/tile1}
  }
  MaterialExpressionTextureSampleParameter2D_1 {
   class=MaterialExpressionTextureSampleParameter2D
   inputs [
    UVs {from=MaterialExpressionComponentMask_2, output=""}
    "DDX(UVs)" {from=MaterialExpressionDDX_0, output=""}
    "DDY(UVs)" {from=MaterialExpressionDDY_0, output=""}
    Apply View MipBias {from="", output=""}
   ]
   properties {
    mipValueMode=TMVM_Derivative
    parameterName=Diffuse
    samplerType=SAMPLERTYPE_LinearColor
    texture=&/tile1_relief_H
   }
  }
  MaterialExpressionTransform_0 {
   class=MaterialExpressionTransform
   inputs [None {from=MaterialExpressionCameraVectorWS_0}]
   properties {transformSourceType=TRANSFORMSOURCE_World, transformType=TRANSFORM_Tangent}
  }
  MaterialExpressionTransform_1 {
   class=MaterialExpressionTransform
   inputs [None {from=MaterialExpressionSubtract_0}]
   properties {transformSourceType=TRANSFORMSOURCE_World, transformType=TRANSFORM_Tangent}
  }
  MaterialExpressionTransform_2 {class=MaterialExpressionTransform, inputs [None {from=MaterialExpressionSubtract_3}]}
  MaterialExpressionTransform_4 {
   class=MaterialExpressionTransform
   inputs [None {from=MaterialExpressionCameraVectorWS_9}]
   properties {transformSourceType=TRANSFORMSOURCE_World, transformType=TRANSFORM_Tangent}
  }
  MaterialExpressionTwoSidedSign_0 {class=MaterialExpressionTwoSidedSign, inputs [], properties {}}
  MaterialExpressionVertexNormalWS_0 {class=MaterialExpressionVertexNormalWS, inputs [], properties {}}
  MaterialExpressionWorldPosition_0 {class=MaterialExpressionWorldPosition, inputs [], properties {}}
  MaterialExpressionWorldPosition_4 {class=MaterialExpressionWorldPosition, inputs [], properties {}}
 }
 properties {bAutomaticallySetUsageInEditor=false, floatPrecisionMode=MFPM_Full, twoSided=true, usedShadingModels=Default Lit}
}
